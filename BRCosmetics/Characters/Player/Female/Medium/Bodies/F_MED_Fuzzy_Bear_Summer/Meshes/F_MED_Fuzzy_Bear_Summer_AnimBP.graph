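# Post-process style Anim Blueprint for the Fuzzy Bear Summer female body cosmetic, adding secondary motion and backbling socket adjustment.
# It copies the attached parent's pose, runs a deformation Control Rig, then simulates four tie bones with sprung AnimDynamics.
# Two bool-gated bone transforms offset the backpack socket for backblings or capes when the Stage2 variant tag marks the hood down.
# Left and right buckle constraint nodes exist but do not feed the output pose.
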AnimBlueprint /BRCosmetics/Characters/Player/Female/Medium/Bodies/F_MED_Fuzzy_Bear_Summer/Meshes/F_MED_Fuzzy_Bear_Summer_AnimBP
target_skeleton=&/F_MED_Fuzzy_Bear_Summer_Skeleton
graphs [
 AnimGraph {nodes [
   n1 = "Constraint - dyn_L_buckleConstraint (PPPPPPPP)"!(
    ComponentPose
    Alpha=1
    ConstraintWeights_0=0.3831
    ConstraintWeights_1:float=0.0251
    ConstraintWeights_2:float=0.0104
    ConstraintWeights_3:float=0.0116
    ConstraintWeights_4:float=0.2445
    ConstraintWeights_5:float=0.05
    ConstraintWeights_6:float=0.0219
    ConstraintWeights_7:float=0.2535
   ) -> (Pose) {
    alphaBoolBlend {blendOption=Linear}
    boneToModify {boneName=dyn_L_buckleConstraint}
    constraintSetup [
     {targetBone {boneName=clavicle_l}}
     {targetBone {boneName=spine_03}}
     {targetBone {boneName=clavicle_r}}
     {targetBone {boneName=deform_upperarm_bk_l}}
     {targetBone {boneName=spine_04}}
     {targetBone {boneName=upperarm_twist_01_l}}
     {targetBone {boneName=spine_02}}
     {targetBone {boneName=spine_05}}
    ]
    constraintWeights [1,1,1,1,1,1,1,1]
   }
   "Constraint - dyn_R_buckleConstraint (PPPPPPPP)"!(
    ComponentPose=@n1
    Alpha=1
    ConstraintWeights_0=0.0125
    ConstraintWeights_1:float=0.0353
    ConstraintWeights_2:float=0.0266
    ConstraintWeights_3:float=0.3732
    ConstraintWeights_4:float=0.2535
    ConstraintWeights_5:float=0.0192
    ConstraintWeights_6:float=0.0251
    ConstraintWeights_7:float=0.2546
   ) -> (Pose) {
    alphaBoolBlend {blendOption=Linear}
    boneToModify {boneName=dyn_R_buckleConstraint}
    constraintSetup [
     {targetBone {boneName=clavicle_l}}
     {targetBone {boneName=upperarm_twist_01_r}}
     {targetBone {boneName=spine_03}}
     {targetBone {boneName=clavicle_r}}
     {targetBone {boneName=spine_04}}
     {targetBone {boneName=deform_upperarm_bk_r}}
     {targetBone {boneName=spine_02}}
     {targetBone {boneName=spine_05}}
    ]
    constraintWeights [1,1,1,1,1,1,1,1]
   }
   n2 = Control Rig(Source=Copy Pose From Mesh() {bUseAttachedParent=true}) {
    alphaBoolBlend {blendOption=Linear}
    controlRigAssetReference {blueprintRigClass=&/F_MED_Fuzzy_Bear_Summer_Deformation_ControlRig.F_MED_Fuzzy_Bear_Summer_Deformation_ControlRig_C}
   }
   n3 = Anim Dynamics - Bone: dyn_L_tieInner(ComponentPose=Local To Component(LocalPose=@n2), Alpha=1, ExternalForce="0,0,0") {
    angularDampingOverride=0.9999
    angularSpringConstant=2000
    bAngularSpring=true
    bOverrideAngularDamping=true
    bUseGravityOverride=true
    gravityScale=0.35
    alphaBoolBlend {blendOption=Linear}
    boundBone {boneName=dyn_L_tieInner}
    componentLinearAccScale=<1 1 1>
    componentLinearVelScale=<1 1 1>
    gravityOverride {z=-1000}
    physicsBodyDefinitions [{
      boundBone {boneName=dyn_L_tieInner}
      boxExtents=<12 12 12>
      constraintSetup {angularLimitsMax=<0 10 0>, angularLimitsMin=<0 -80 0>, angularTarget=<0 0 0>}
      localJointOffset=<3 0 0>
     }]
   }
   n4 = Anim Dynamics - Bone: dyn_L_tieOuter(ComponentPose=@n3, Alpha=1, ExternalForce="0,0,0") {
    angularDampingOverride=0.9999
    angularSpringConstant=2000
    bAngularSpring=true
    bOverrideAngularDamping=true
    bUseGravityOverride=true
    gravityScale=0.35
    alphaBoolBlend {blendOption=Linear}
    boundBone {boneName=dyn_L_tieOuter}
    componentLinearAccScale=<1 1 1>
    componentLinearVelScale=<1 1 1>
    gravityOverride {z=-1000}
    physicsBodyDefinitions [{
      boundBone {boneName=dyn_L_tieOuter}
      boxExtents=<12 12 12>
      constraintSetup {angularLimitsMax=<0 10 0>, angularLimitsMin=<0 -80 0>, angularTarget=<0 0 0>}
      localJointOffset=<3 0 0>
     }]
   }
   n5 = Anim Dynamics - Bone: dyn_R_tieInner(ComponentPose=@n4, Alpha=1, ExternalForce="0,0,0") {
    angularDampingOverride=0.9999
    angularSpringConstant=2000
    bAngularSpring=true
    bOverrideAngularDamping=true
    bUseGravityOverride=true
    gravityScale=0.35
    alphaBoolBlend {blendOption=Linear}
    boundBone {boneName=dyn_R_tieInner}
    componentLinearAccScale=<1 1 1>
    componentLinearVelScale=<1 1 1>
    gravityOverride {z=-1000}
    physicsBodyDefinitions [{
      boundBone {boneName=dyn_R_tieInner}
      boxExtents=<12 12 12>
      constraintSetup {angularLimitsMax=<0 10 0>, angularLimitsMin=<0 -80 0>, angularTarget=<0 0 0>}
      localJointOffset=<3 0 0>
     }]
   }
   n6 = Anim Dynamics - Bone: dyn_R_tieOuter(ComponentPose=@n5, Alpha=1, ExternalForce="0,0,0") {
    angularDampingOverride=0.9999
    angularSpringConstant=2000
    bAngularSpring=true
    bOverrideAngularDamping=true
    bUseGravityOverride=true
    gravityScale=0.35
    alphaBoolBlend {blendOption=Linear}
    boundBone {boneName=dyn_R_tieOuter}
    componentLinearAccScale=<1 1 1>
    componentLinearVelScale=<1 1 1>
    gravityOverride {z=-1000}
    physicsBodyDefinitions [{
      boundBone {boneName=dyn_R_tieOuter}
      boxExtents=<12 12 12>
      constraintSetup {angularLimitsMax=<0 10 0>, angularLimitsMin=<0 -80 0>, angularTarget=<0 0 0>}
      localJointOffset=<3 0 0>
     }]
   }
   n7 = "Transform (Modify) Bone - Bone: dyn_backpackSocket"!(
    ComponentPose=@n6
    bAlphaBoolEnabled:bool=Get scootBackblingSocket() -> (scootBackblingSocket:bool)
    Translation="-0.000,-6,0.000"
    Rotation="0,0,0"
    Scale="1,1,1"
   ) -> (Pose) {
    alphaInputType=Bool
    rotationMode=BMM_Replace
    rotationSpace=BCS_BoneSpace
    translationMode=BMM_Replace
    translationSpace=BCS_BoneSpace
    alphaBoolBlend {blendOption=Linear}
    boneToModify {boneName=dyn_backpackSocket}
    translation {y=-6}
   }
   n8 = "Transform (Modify) Bone - Bone: dyn_backpackSocket"(ComponentPose=@n7, bAlphaBoolEnabled=Get scootCapeSocket() -> (scootCapeSocket:bool), Translation="0,-6,0", Rotation="0,0,0", Scale="1,1,1") {
    alphaInputType=Bool
    rotationMode=BMM_Additive
    rotationSpace=BCS_BoneSpace
    translationMode=BMM_Replace
    translationSpace=BCS_BoneSpace
    alphaBoolBlend {blendOption=Linear}
    boneToModify {boneName=dyn_backpackSocket}
    translation {y=-6}
   }
   Output Pose(Result=Component To Local(ComponentPose=@n8)) {layerGroup=None}
  ]}
 EventGraph {nodes [
   Event BlueprintUpdateAnimation()
   > n1 = Sequence()
   n2 = CheckForBackbling()
   > n3 = Set Is Wearing Backbling(Is Wearing Backbling=@n2)
   n4 = Get Is Backbling A Cape() -> (Is Backbling A Cape:bool)
   > Set isWearingBackblingNotA Cape(isWearingBackblingNotA Cape:bool=AND Boolean(A=@n3, B=NOT Boolean(A=@n4))) -> (Output_Get:bool)
   Set Is Backbling A Cape(execute=@n1.then_1, Is Backbling A Cape:bool=@n2.bIsBackblingACape) -> (Output_Get:bool)
   > Set scootBackblingSocket(scootBackblingSocket:bool=AND Boolean(A=Get hoodDown?() -> (hoodDown?:bool), B=Get isWearingBackblingNotA Cape() -> (isWearingBackblingNotA Cape:bool))) -> (Output_Get:bool)
   > Set scootCapeSocket(scootCapeSocket:bool=AND Boolean(A=Get hoodDown?(), B=Get Is Backbling A Cape())) -> (Output_Get:bool)
   Event OnCharacterCustomizationLoaded()
   > n5 = Branch(Condition=CharacterItemDefinitionHasAnyMetaTag(Tags=Make Array("[0]"=(TagName="Cosmetics.Variant.Property.Stage2"))))
   > Set hoodDown?(hoodDown?:bool=true) -> (Output_Get:bool)
   Set hoodDown?(execute=@n5.else, hoodDown?=false)
  ]}
]
variables [
 hoodDown? {type=boolean}
 Is Backbling A Cape {type=boolean}
 Is Wearing Backbling {type=boolean}
 isWearingBackblingNotA Cape {type=boolean}
 scootBackblingSocket {type=boolean}
 scootCapeSocket {type=boolean}
]
