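# Physics asset for the Fuzzy Bear Summer medium female cosmetic's back face-accessory part, covering back hair and a drooped hood.
# Capsules on the spine, neck, head, clavicles and upper arms form the body, with the hood linked to spine_05 by free angular motion.
# Left, center and right hair chains hang from head-linked Lookat bodies, with swing limits widening toward the tips.
# Angularly free cross-constraints tie the side chains to the center one.
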PhysicsAsset /BRCosmetics/Characters/Player/Female/Medium/Bodies/F_MED_Fuzzy_Bear_Summer/Meshes/Parts/F_MED_Fuzzy_Bear_Summer_Back_FaceAcc_Physics
bodies {
 clavicle_l [{
   shapeType=Capsule
   length=7.5
   radius=6.5
   shapeName=clavicle_l_capsule
   center=<5.0116 -1 0.5498>
   rotation=<rot 87.8504 -179.9983 -179.9983>
  }]
 clavicle_r [{shapeType=Capsule, length=8.25, radius=5.75, shapeName=clavicle_l_capsule, center=<-5 1 0>, rotation=<rot 90 0 0>}]
 dyn_C_hairBack_0 [{shapeType=Capsule, length=5, radius=3, shapeName=dyn_C_hairBack_0_capsule, center=<5 0 0>, rotation=<rot 90 0 0>}]
 dyn_C_hairBack_1 [{shapeType=Capsule, length=4.6042, radius=4.0031, shapeName=dyn_C_hairBack_1_capsule, center=<5 0 0>, rotation=<rot 90 0 0>}]
 dyn_C_hairBack_2 [{shapeType=Capsule, length=6, radius=4, shapeName=dyn_C_hairBack_2_capsule, center=<5 0 1.5>, rotation=<rot 90 0 0>}]
 dyn_C_hairBack_Lookat [{shapeType=Capsule, length=0.505, radius=0.505, shapeName=dyn_C_hairBack_Lookat_capsule, rotation=<rot 0 -90 0>}]
 dyn_C_hairBack_strand [{shapeType=Capsule, length=5.4118, radius=3.442, shapeName=dyn_C_hairBack_strand_capsule, center=<5 0 0>, rotation=<rot 90 0 0>}]
 dyn_L_hairBack_0 [{shapeType=Capsule, length=5, radius=3, shapeName=dyn_L_hairBack_0_capsule, center=<5 0 0>, rotation=<rot 90 0 0>}]
 dyn_L_hairBack_1 [
  {shapeType=Capsule, length=2.75, radius=5, shapeName=dyn_L_hairBack_1_capsule, center=<5 0 1>, rotation=<rot 90 0 0>}
  {
   shapeType=Capsule
   length=7
   radius=2.75
   shapeName=dyn_L_hairBack_1_capsule
   center=<3.7754 3.7015 -1.583>
   rotation=<rot 85 -164.9999 -179.9999>
  }
 ]
 dyn_L_hairBack_2 [
  {shapeType=Capsule, length=5.75, radius=4.25, shapeName=dyn_L_hairBack_2_capsule, center=<5 0 1.5>, rotation=<rot 90 0 0>}
  {shapeType=Capsule, length=5, radius=3.25, shapeName=dyn_L_hairBack_2_capsule, center=<3.8695 4.0313 -0.384>, rotation=<rot 85 0 0>}
 ]
 dyn_L_hairBack_Lookat [{shapeType=Capsule, length=0.505, radius=0.505, shapeName=dyn_L_hairBack_Lookat_capsule, rotation=<rot 0 -90 0>}]
 dyn_L_hairBack_strand [{shapeType=Capsule, length=4.5339, radius=2.878, shapeName=dyn_L_hairBack_strand_capsule, center=<5 0 0>, rotation=<rot 90 0 0>}]
 dyn_R_hairBack_0 [{shapeType=Capsule, length=5, radius=3, shapeName=dyn_R_hairBack_0_capsule, center=<5 0 0>, rotation=<rot 90 0 0>}]
 dyn_R_hairBack_1 [{shapeType=Capsule, length=3.5, radius=4.25, shapeName=dyn_R_hairBack_1_capsule, center=<5 0 0>, rotation=<rot 90 0 0>}]
 dyn_R_hairBack_2 [
  {shapeType=Capsule, length=5.5, radius=4.5, shapeName=dyn_R_hairBack_2_capsule, center=<5 0 1.5>, rotation=<rot 90 0 0>}
  {
   shapeType=Capsule
   length=9.5
   radius=1.75
   shapeName=dyn_R_hairBack_2_capsule
   center=<3.4806 -3.3363 -1.0109>
   rotation=<rot 80.0001 180 180>
  }
 ]
 dyn_R_hairBack_Lookat [{shapeType=Capsule, length=0.505, radius=0.505, shapeName=dyn_R_hairBack_Lookat_capsule, rotation=<rot 0 -90 0>}]
 dyn_R_hairBack_strand [{shapeType=Capsule, length=5.0338, radius=3.3974, shapeName=dyn_R_hairBack_strand_capsule, center=<5 0 0>, rotation=<rot 90 0 0>}]
 dyn_droopedHood [
  {shapeType=Capsule, length=16.0714, radius=5.1377, shapeName=dyn_droopedHood_capsule, center=<5.5245 2.1831 0>}
  {shapeType=Capsule, length=12.25, radius=5.25, shapeName=dyn_droopedHood_capsule, center=<10.889 2.8964 0>}
  {shapeType=Capsule, length=14.5, radius=4.25, shapeName=dyn_droopedHood_capsule, center=<16.3314 3.4244 0>}
 ]
 head [{shapeType=Capsule, length=2.75, radius=7.25, shapeName=head_capsule, center=<0.7552 2.7733 0>, rotation=<rot 90 0 0>}]
 neck_01 [{shapeType=Capsule, length=1, radius=5.25, shapeName=neck_01_capsule, center=<1.214 0 0>, rotation=<rot 90 0 0>}]
 spine_04 [
  {shapeType=Capsule, length=8, radius=9.5, shapeName=spine_04_capsule, center=<-1.496 3.568 0>}
  {shapeType=Capsule, length=12.75, radius=8, shapeName=spine_04_capsule, center=<5.649 3.4316 0>}
 ]
 spine_05 [{shapeType=Capsule, length=13.75, radius=9, shapeName=spine_05_capsule, center=<0.7703 -0.5143 0>}]
 upperarm_l [{shapeType=Capsule, length=17.25, radius=3.5, shapeName=upperarm_l_capsule, center=<10.5389 0.0414 2.7917>, rotation=<rot 90 0 0>}]
 upperarm_r [{shapeType=Capsule, length=17.75, radius=3, shapeName=upperarm_r_capsule, center=<-10.5389 0.3624 -3.2892>, rotation=<rot 90 0 0>}]
}
constraints [
 {
  bone1Name=spine_05
  bone2Name=spine_04
  swing1LimitDegrees=45
  swing1Motion=Free
  swing2LimitDegrees=45
  swing2Motion=Free
  twistLimitDegrees=45
  twistMotion=Free
 }
 {
  bone1Name=clavicle_l
  bone2Name=spine_05
  swing1LimitDegrees=45
  swing1Motion=Free
  swing2LimitDegrees=45
  swing2Motion=Free
  twistLimitDegrees=45
  twistMotion=Free
 }
 {
  bone1Name=clavicle_r
  bone2Name=spine_05
  swing1LimitDegrees=45
  swing1Motion=Free
  swing2LimitDegrees=45
  swing2Motion=Free
  twistLimitDegrees=45
  twistMotion=Free
 }
 {
  bone1Name=head
  bone2Name=neck_01
  swing1LimitDegrees=45
  swing1Motion=Free
  swing2LimitDegrees=45
  swing2Motion=Free
  twistLimitDegrees=45
  twistMotion=Free
 }
 {
  bone1Name=neck_01
  bone2Name=spine_05
  swing1LimitDegrees=45
  swing1Motion=Free
  swing2LimitDegrees=45
  swing2Motion=Free
  twistLimitDegrees=45
  twistMotion=Free
 }
 {
  bone1Name=upperarm_l
  bone2Name=clavicle_l
  swing1LimitDegrees=45
  swing1Motion=Free
  swing2LimitDegrees=45
  swing2Motion=Free
  twistLimitDegrees=45
  twistMotion=Free
 }
 {
  bone1Name=upperarm_r
  bone2Name=clavicle_r
  swing1LimitDegrees=45
  swing1Motion=Free
  swing2LimitDegrees=45
  swing2Motion=Free
  twistLimitDegrees=45
  twistMotion=Free
 }
 {
  bone1Name=dyn_droopedHood
  bone2Name=spine_05
  swing1LimitDegrees=45
  swing1Motion=Free
  swing2LimitDegrees=45
  swing2Motion=Free
  twistLimitDegrees=45
  twistMotion=Free
 }
 {
  bone1Name=dyn_R_hairBack_Lookat
  bone2Name=head
  swing1LimitDegrees=45
  swing1Motion=Free
  swing2LimitDegrees=45
  swing2Motion=Free
  twistLimitDegrees=45
  twistMotion=Free
 }
 {bone1Name=dyn_R_hairBack_0, bone2Name=dyn_R_hairBack_Lookat, swing1LimitDegrees=15, swing2LimitDegrees=15, twistLimitDegrees=5}
 {bone1Name=dyn_R_hairBack_1, bone2Name=dyn_R_hairBack_0, swing1LimitDegrees=15, swing2LimitDegrees=20, twistLimitDegrees=5}
 {bone1Name=dyn_R_hairBack_2, bone2Name=dyn_R_hairBack_1, swing1LimitDegrees=15, swing2LimitDegrees=35, twistLimitDegrees=15}
 {bone1Name=dyn_R_hairBack_strand, bone2Name=dyn_R_hairBack_2, swing1LimitDegrees=15, swing2LimitDegrees=45, twistLimitDegrees=15}
 {
  bone1Name=dyn_C_hairBack_Lookat
  bone2Name=head
  swing1LimitDegrees=45
  swing1Motion=Free
  swing2LimitDegrees=45
  swing2Motion=Free
  twistLimitDegrees=45
  twistMotion=Free
 }
 {bone1Name=dyn_C_hairBack_0, bone2Name=dyn_C_hairBack_Lookat, swing1LimitDegrees=15, swing2LimitDegrees=15, twistLimitDegrees=5}
 {bone1Name=dyn_C_hairBack_1, bone2Name=dyn_C_hairBack_0, swing1LimitDegrees=15, swing2LimitDegrees=20, twistLimitDegrees=5}
 {bone1Name=dyn_C_hairBack_2, bone2Name=dyn_C_hairBack_1, swing1LimitDegrees=15, swing2LimitDegrees=35, twistLimitDegrees=15}
 {bone1Name=dyn_C_hairBack_strand, bone2Name=dyn_C_hairBack_2, swing1LimitDegrees=15, swing2LimitDegrees=45, twistLimitDegrees=15}
 {
  bone1Name=dyn_L_hairBack_Lookat
  bone2Name=head
  swing1LimitDegrees=45
  swing1Motion=Free
  swing2LimitDegrees=45
  swing2Motion=Free
  twistLimitDegrees=45
  twistMotion=Free
 }
 {bone1Name=dyn_L_hairBack_0, bone2Name=dyn_L_hairBack_Lookat, swing1LimitDegrees=15, swing2LimitDegrees=15, twistLimitDegrees=5}
 {bone1Name=dyn_L_hairBack_1, bone2Name=dyn_L_hairBack_0, swing1LimitDegrees=15, swing2LimitDegrees=20, twistLimitDegrees=5}
 {bone1Name=dyn_L_hairBack_2, bone2Name=dyn_L_hairBack_1, swing1LimitDegrees=15, swing2LimitDegrees=35, twistLimitDegrees=15}
 {bone1Name=dyn_L_hairBack_strand, bone2Name=dyn_L_hairBack_2, swing1LimitDegrees=15, swing2LimitDegrees=45, twistLimitDegrees=15}
 {
  bone1Name=dyn_L_hairBack_2
  bone2Name=dyn_C_hairBack_2
  swing1LimitDegrees=45
  swing1Motion=Free
  swing2LimitDegrees=45
  swing2Motion=Free
  twistLimitDegrees=45
  twistMotion=Free
 }
 {
  bone1Name=dyn_R_hairBack_2
  bone2Name=dyn_C_hairBack_2
  swing1LimitDegrees=45
  swing1Motion=Free
  swing2LimitDegrees=45
  swing2Motion=Free
  twistLimitDegrees=45
  twistMotion=Free
 }
 {
  bone1Name=dyn_R_hairBack_1
  bone2Name=dyn_C_hairBack_1
  swing1LimitDegrees=45
  swing1Motion=Free
  swing2LimitDegrees=45
  swing2Motion=Free
  twistLimitDegrees=45
  twistMotion=Free
 }
 {
  bone1Name=dyn_L_hairBack_1
  bone2Name=dyn_C_hairBack_1
  swing1LimitDegrees=45
  swing1Motion=Free
  swing2LimitDegrees=45
  swing2Motion=Free
  twistLimitDegrees=45
  twistMotion=Free
 }
 {
  bone1Name=dyn_L_hairBack_strand
  bone2Name=dyn_C_hairBack_strand
  swing1LimitDegrees=45
  swing1Motion=Free
  swing2LimitDegrees=45
  swing2Motion=Free
  twistLimitDegrees=45
  twistMotion=Free
 }
 {
  bone1Name=dyn_R_hairBack_strand
  bone2Name=dyn_C_hairBack_strand
  swing1LimitDegrees=45
  swing1Motion=Free
  swing2LimitDegrees=45
  swing2Motion=Free
  twistLimitDegrees=45
  twistMotion=Free
 }
]
