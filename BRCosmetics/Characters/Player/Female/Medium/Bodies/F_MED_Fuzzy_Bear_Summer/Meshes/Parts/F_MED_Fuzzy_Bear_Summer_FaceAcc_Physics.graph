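# Physics asset for the Fuzzy Bear Summer female medium character's FaceAcc mesh, giving physics shapes to its hair and drooped hood.
# Capsules on upper spine, neck, head, clavicles and upper arms form the body, joined by free-rotating constraints.
# Mirrored four-link front hair chains hang from neck_01, back chains from head, and a box hood from spine_05.
# These dynamic joints lock twist; back hair swings freely below its limited root.
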PhysicsAsset /BRCosmetics/Characters/Player/Female/Medium/Bodies/F_MED_Fuzzy_Bear_Summer/Meshes/Parts/F_MED_Fuzzy_Bear_Summer_FaceAcc_Physics
bodies {
 clavicle_l [{shapeType=Capsule, length=20, radius=7, shapeName=clavicle_l_capsule, center=<4.9552 -1.1362 0.0975>, rotation=<rot 90 0 0>}]
 clavicle_r [{shapeType=Capsule, length=20, radius=7, shapeName=clavicle_l_capsule, center=<-4.9552 1.1362 -0.0975>, rotation=<rot 90 0 0>}]
 dyn_L_hairBack_0 [{shapeType=Capsule, length=6, radius=5, shapeName=dyn_L_hairBack_0_capsule, center=<-3.6775 0 0>, rotation=<rot 90 0 0>}]
 dyn_L_hairBack_1 [{shapeType=Capsule, length=13, radius=5.5, shapeName=dyn_L_hairBack_1_capsule, center=<-3.259 0 0>, rotation=<rot 90 0 0>}]
 dyn_L_hairBack_2 [{shapeType=Capsule, length=13, radius=4, shapeName=dyn_L_hairBack_2_capsule, center=<-1.8651 0 0>, rotation=<rot 90 0 0>}]
 dyn_L_hairBack_strand [{shapeType=Capsule, length=4, radius=4, shapeName=dyn_L_hairBack_strand_capsule, center=<-1.8651 0 0>, rotation=<rot 90 0 0>}]
 dyn_L_hairFront_0 [{shapeType=Capsule, length=7, radius=2, shapeName=dyn_L_hairFront_0_capsule, center=<3.2434 0 0>, rotation=<rot 90 0 0>}]
 dyn_L_hairFront_1 [{shapeType=Capsule, length=7, radius=2.5, shapeName=dyn_R_hairFront_1_capsule, center=<2 0 0.5>, rotation=<rot 90 0 0>}]
 dyn_L_hairFront_2 [{shapeType=Capsule, length=5, radius=2.8, shapeName=dyn_R_hairFront_2_capsule, center=<2 0 -0.3>, rotation=<rot 90 0 0>}]
 dyn_L_hairFront_3 [{shapeType=Capsule, length=5, radius=2, shapeName=dyn_R_hairFront_2_capsule, center=<1 0 0>, rotation=<rot 90 0 0>}]
 dyn_R_hairBack_0 [{shapeType=Capsule, length=6, radius=5, shapeName=dyn_R_hairBack_0_capsule, center=<3.6775 0 0>, rotation=<rot 90 0 0>}]
 dyn_R_hairBack_1 [{shapeType=Capsule, length=13, radius=5.5, shapeName=dyn_R_hairBack_1_capsule, center=<3.259 0 0>, rotation=<rot 90 0 0>}]
 dyn_R_hairBack_2 [{shapeType=Capsule, length=13, radius=4, shapeName=dyn_R_hairBack_2_capsule, center=<1.8651 0 0>, rotation=<rot 90 0 0>}]
 dyn_R_hairBack_strand [{shapeType=Capsule, length=4, radius=4, shapeName=dyn_R_hairBack_strand_capsule, center=<1.8651 0 0>, rotation=<rot 90 0 0>}]
 dyn_R_hairFront_0 [{shapeType=Capsule, length=7, radius=2, shapeName=dyn_R_hairFront_0_capsule, center=<-3.2434 0 0>, rotation=<rot 90 0 0>}]
 dyn_R_hairFront_1 [{shapeType=Capsule, length=7, radius=2.5, shapeName=dyn_R_hairFront_1_capsule, center=<-2 0 -0.5>, rotation=<rot 90 0 0>}]
 dyn_R_hairFront_2 [{shapeType=Capsule, length=4, radius=2.5, shapeName=dyn_R_hairFront_2_capsule, center=<-2 0 0.3>, rotation=<rot 90 0 0>}]
 dyn_R_hairFront_3 [{shapeType=Capsule, length=5, radius=2, shapeName=dyn_R_hairFront_2_capsule, center=<-1 0 0>, rotation=<rot 90 0 0>}]
 dyn_droopedHood [{shapeType=Box, extentX=17.4556, extentY=6.1592, extentZ=31.402, shapeName=dyn_droopedHood_box, center=<10 0 0>}]
 head [{shapeType=Capsule, length=1, radius=1, shapeName=head_capsule, rotation=<rot 90 0 0>}]
 neck_01 [{shapeType=Capsule, length=10, radius=2, shapeName=neck_01_capsule, rotation=<rot 90 0 0>}]
 spine_04 [{shapeType=Capsule, length=3, radius=12, shapeName=spine_04_capsule, center=<-1.496 3.2312 0>}]
 spine_05 [{shapeType=Capsule, length=49.25, radius=10, shapeName=spine_05_capsule, center=<2.3404 -2.3543 0>}]
 upperarm_l [{shapeType=Capsule, length=30, radius=2, shapeName=upperarm_l_capsule, center=<10.5389 -1.3484 -0.0193>, rotation=<rot 90 0 0>}]
 upperarm_r [{shapeType=Capsule, length=30, radius=2, shapeName=upperarm_r_capsule, center=<-10.5389 1.3484 0.0193>, rotation=<rot 90 0 0>}]
}
constraints [
 {
  bone1Name=dyn_L_hairFront_2
  bone2Name=dyn_L_hairFront_1
  swing1LimitDegrees=30
  swing2LimitDegrees=30
  twistLimitDegrees=30
  twistMotion=Locked
 }
 {
  bone1Name=dyn_R_hairFront_2
  bone2Name=dyn_R_hairFront_1
  swing1LimitDegrees=30
  swing2LimitDegrees=30
  twistLimitDegrees=30
  twistMotion=Locked
 }
 {
  bone1Name=dyn_L_hairFront_1
  bone2Name=dyn_L_hairFront_0
  swing1LimitDegrees=30
  swing2LimitDegrees=30
  twistLimitDegrees=30
  twistMotion=Locked
 }
 {
  bone1Name=dyn_R_hairFront_1
  bone2Name=dyn_R_hairFront_0
  swing1LimitDegrees=30
  swing2LimitDegrees=30
  twistLimitDegrees=30
  twistMotion=Locked
 }
 {
  bone1Name=spine_05
  bone2Name=spine_04
  swing1LimitDegrees=45
  swing1Motion=Free
  swing2LimitDegrees=45
  swing2Motion=Free
  twistLimitDegrees=45
  twistMotion=Free
 }
 {
  bone1Name=clavicle_l
  bone2Name=spine_05
  swing1LimitDegrees=45
  swing1Motion=Free
  swing2LimitDegrees=45
  swing2Motion=Free
  twistLimitDegrees=45
  twistMotion=Free
 }
 {
  bone1Name=clavicle_r
  bone2Name=spine_05
  swing1LimitDegrees=45
  swing1Motion=Free
  swing2LimitDegrees=45
  swing2Motion=Free
  twistLimitDegrees=45
  twistMotion=Free
 }
 {
  bone1Name=dyn_L_hairFront_3
  bone2Name=dyn_L_hairFront_2
  swing1LimitDegrees=30
  swing2LimitDegrees=30
  twistLimitDegrees=30
  twistMotion=Locked
 }
 {
  bone1Name=dyn_R_hairFront_3
  bone2Name=dyn_R_hairFront_2
  swing1LimitDegrees=30
  swing2LimitDegrees=30
  twistLimitDegrees=30
  twistMotion=Locked
 }
 {
  bone1Name=head
  bone2Name=neck_01
  swing1LimitDegrees=45
  swing1Motion=Free
  swing2LimitDegrees=45
  swing2Motion=Free
  twistLimitDegrees=45
  twistMotion=Free
 }
 {
  bone1Name=neck_01
  bone2Name=spine_05
  swing1LimitDegrees=45
  swing1Motion=Free
  swing2LimitDegrees=45
  swing2Motion=Free
  twistLimitDegrees=45
  twistMotion=Free
 }
 {
  bone1Name=upperarm_l
  bone2Name=clavicle_l
  swing1LimitDegrees=45
  swing1Motion=Free
  swing2LimitDegrees=45
  swing2Motion=Free
  twistLimitDegrees=45
  twistMotion=Free
 }
 {
  bone1Name=upperarm_r
  bone2Name=clavicle_r
  swing1LimitDegrees=45
  swing1Motion=Free
  swing2LimitDegrees=45
  swing2Motion=Free
  twistLimitDegrees=45
  twistMotion=Free
 }
 {bone1Name=dyn_droopedHood, bone2Name=spine_05, swing1LimitDegrees=45, swing2LimitDegrees=10, twistLimitDegrees=45, twistMotion=Locked}
 {bone1Name=dyn_L_hairFront_0, bone2Name=neck_01, swing1LimitDegrees=30, swing2LimitDegrees=30, twistLimitDegrees=30, twistMotion=Locked}
 {bone1Name=dyn_R_hairFront_0, bone2Name=neck_01, swing1LimitDegrees=30, swing2LimitDegrees=30, twistLimitDegrees=30, twistMotion=Locked}
 {bone1Name=dyn_R_hairBack_0, bone2Name=head, swing1LimitDegrees=30, swing2LimitDegrees=30, twistLimitDegrees=45, twistMotion=Locked}
 {
  bone1Name=dyn_R_hairBack_1
  bone2Name=dyn_R_hairBack_0
  swing1LimitDegrees=20
  swing1Motion=Free
  swing2LimitDegrees=30
  swing2Motion=Free
  twistLimitDegrees=45
  twistMotion=Locked
 }
 {
  bone1Name=dyn_R_hairBack_2
  bone2Name=dyn_R_hairBack_1
  swing1LimitDegrees=20
  swing1Motion=Free
  swing2LimitDegrees=30
  swing2Motion=Free
  twistLimitDegrees=45
  twistMotion=Locked
 }
 {
  bone1Name=dyn_R_hairBack_strand
  bone2Name=dyn_R_hairBack_2
  swing1LimitDegrees=20
  swing1Motion=Free
  swing2LimitDegrees=30
  swing2Motion=Free
  twistLimitDegrees=45
  twistMotion=Locked
 }
 {bone1Name=dyn_L_hairBack_0, bone2Name=head, swing1LimitDegrees=30, swing2LimitDegrees=30, twistLimitDegrees=45, twistMotion=Locked}
 {
  bone1Name=dyn_L_hairBack_1
  bone2Name=dyn_L_hairBack_0
  swing1LimitDegrees=20
  swing1Motion=Free
  swing2LimitDegrees=30
  swing2Motion=Free
  twistLimitDegrees=45
  twistMotion=Locked
 }
 {
  bone1Name=dyn_L_hairBack_2
  bone2Name=dyn_L_hairBack_1
  swing1LimitDegrees=20
  swing1Motion=Free
  swing2LimitDegrees=30
  swing2Motion=Free
  twistLimitDegrees=45
  twistMotion=Locked
 }
 {
  bone1Name=dyn_L_hairBack_strand
  bone2Name=dyn_L_hairBack_2
  swing1LimitDegrees=20
  swing1Motion=Free
  swing2LimitDegrees=30
  swing2Motion=Free
  twistLimitDegrees=45
  twistMotion=Locked
 }
]
