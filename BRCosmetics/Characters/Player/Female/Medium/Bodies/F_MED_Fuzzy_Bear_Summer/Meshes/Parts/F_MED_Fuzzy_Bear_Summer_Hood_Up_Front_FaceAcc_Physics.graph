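# Physics asset for the hood-up front face-accessory part of the female medium Fuzzy Bear Summer body, covering its left and right front hair dynamic bones.
# Capsules approximate the upper spine, neck, head, clavicles and upper arms, joined by free-swing, free-twist constraints.
# Each hair chain links to the neck through a Lookat bone with free rotation, then tighter angular limits on its three outer segments.
# Those segments use capsules on the left but boxes on the right.
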PhysicsAsset /BRCosmetics/Characters/Player/Female/Medium/Bodies/F_MED_Fuzzy_Bear_Summer/Meshes/Parts/F_MED_Fuzzy_Bear_Summer_Hood_Up_Front_FaceAcc_Physics
bodies {
 clavicle_l [{shapeType=Capsule, length=8.25, radius=5.75, shapeName=clavicle_l_capsule, center=<5 -1 -1>, rotation=<rot 90 0 0>}]
 clavicle_r [{shapeType=Capsule, length=8.25, radius=5.75, shapeName=clavicle_l_capsule, center=<-5 1 0>, rotation=<rot 90 0 0>}]
 dyn_L_hairFront_0 [{shapeType=Capsule, length=0.505, radius=0.505, shapeName=dyn_L_hairFront_0_capsule}]
 dyn_L_hairFront_1 [{shapeType=Capsule, length=4, radius=4, shapeName=dyn_L_hairFront_1_capsule, center=<4 0 2>, rotation=<rot 90 0 0>}]
 dyn_L_hairFront_2 [{shapeType=Capsule, length=4, radius=4, shapeName=dyn_L_hairFront_2_capsule, center=<4 0 2>, rotation=<rot 90 0 0>}]
 dyn_L_hairFront_3 [{shapeType=Capsule, length=4, radius=4, shapeName=dyn_L_hairFront_3_capsule, center=<6 0 2>, rotation=<rot 9.7024 99.9914 113.4548>}]
 dyn_L_hairFront_Lookat [{shapeType=Capsule, length=0.505, radius=0.505, shapeName=dyn_L_hairFront_Lookat_capsule}]
 dyn_R_hairFront_0 [{shapeType=Capsule, length=0.505, radius=0.505, shapeName=dyn_R_hairFront_0_capsule}]
 dyn_R_hairFront_1 [{shapeType=Box, extentX=6.0806, extentY=2.3347, extentZ=8.5, shapeName=dyn_R_hairFront_1_box, center=<3 0 0>, rotation=<rot 0 -0.0002 8.1858>}]
 dyn_R_hairFront_2 [{
   shapeType=Box
   extentX=6.5394
   extentY=2.7618
   extentZ=6
   shapeName=dyn_R_hairFront_2_box
   center=<3.5 0 -0.25>
   rotation=<rot -0.0001 -0.0002 10.7256>
  }]
 dyn_R_hairFront_3 [{
   shapeType=Box
   extentX=8.3117
   extentY=6.1874
   extentZ=2.2257
   shapeName=dyn_R_hairFront_3_box
   center=<3.5 0 -0.5>
   rotation=<rot 13.5541 -0.0001 10.8638>
  }]
 dyn_R_hairFront_Lookat [{shapeType=Capsule, length=0.505, radius=0.505, shapeName=dyn_R_hairFront_Lookat_capsule}]
 head [{shapeType=Capsule, length=1, radius=1, shapeName=head_capsule, rotation=<rot 90 0 0>}]
 neck_01 [{shapeType=Capsule, length=1, radius=5.25, shapeName=neck_01_capsule, center=<1.214 0 0>, rotation=<rot 90 0 0>}]
 spine_04 [
  {shapeType=Capsule, length=4.5, radius=10, shapeName=spine_04_capsule, center=<-0.9596 5.5009 -0.4204>}
  {shapeType=Capsule, length=8.25, radius=8.5, shapeName=spine_04_capsule, center=<3.8698 3.9868 -0.3948>}
 ]
 spine_05 [{shapeType=Capsule, length=17.25, radius=9, shapeName=spine_05_capsule, center=<0.1342 0.3153 0>}]
 upperarm_l [{shapeType=Capsule, length=17.25, radius=3.5, shapeName=upperarm_l_capsule, center=<10.5389 -2 2.7917>, rotation=<rot 90 0 0>}]
 upperarm_r [{
   shapeType=Capsule
   length=16.75
   radius=4
   shapeName=upperarm_r_capsule
   center=<-10.9441 2.6491 -3.4514>
   rotation=<rot 60.2119 -179.9999 -180>
  }]
}
constraints [
 {
  bone1Name=spine_05
  bone2Name=spine_04
  swing1LimitDegrees=45
  swing1Motion=Free
  swing2LimitDegrees=45
  swing2Motion=Free
  twistLimitDegrees=45
  twistMotion=Free
 }
 {
  bone1Name=clavicle_l
  bone2Name=spine_05
  swing1LimitDegrees=45
  swing1Motion=Free
  swing2LimitDegrees=45
  swing2Motion=Free
  twistLimitDegrees=45
  twistMotion=Free
 }
 {
  bone1Name=clavicle_r
  bone2Name=spine_05
  swing1LimitDegrees=45
  swing1Motion=Free
  swing2LimitDegrees=45
  swing2Motion=Free
  twistLimitDegrees=45
  twistMotion=Free
 }
 {
  bone1Name=head
  bone2Name=neck_01
  swing1LimitDegrees=45
  swing1Motion=Free
  swing2LimitDegrees=45
  swing2Motion=Free
  twistLimitDegrees=45
  twistMotion=Free
 }
 {
  bone1Name=neck_01
  bone2Name=spine_05
  swing1LimitDegrees=45
  swing1Motion=Free
  swing2LimitDegrees=45
  swing2Motion=Free
  twistLimitDegrees=45
  twistMotion=Free
 }
 {
  bone1Name=upperarm_l
  bone2Name=clavicle_l
  swing1LimitDegrees=45
  swing1Motion=Free
  swing2LimitDegrees=45
  swing2Motion=Free
  twistLimitDegrees=45
  twistMotion=Free
 }
 {
  bone1Name=upperarm_r
  bone2Name=clavicle_r
  swing1LimitDegrees=45
  swing1Motion=Free
  swing2LimitDegrees=45
  swing2Motion=Free
  twistLimitDegrees=45
  twistMotion=Free
 }
 {
  bone1Name=dyn_L_hairFront_Lookat
  bone2Name=neck_01
  swing1LimitDegrees=45
  swing1Motion=Free
  swing2LimitDegrees=45
  swing2Motion=Free
  twistLimitDegrees=45
  twistMotion=Free
 }
 {
  bone1Name=dyn_L_hairFront_0
  bone2Name=dyn_L_hairFront_Lookat
  swing1LimitDegrees=45
  swing1Motion=Free
  swing2LimitDegrees=45
  swing2Motion=Free
  twistLimitDegrees=45
  twistMotion=Free
 }
 {bone1Name=dyn_L_hairFront_1, bone2Name=dyn_L_hairFront_0, swing1LimitDegrees=15, swing2LimitDegrees=15, twistLimitDegrees=5}
 {bone1Name=dyn_L_hairFront_2, bone2Name=dyn_L_hairFront_1, swing1LimitDegrees=35, swing2LimitDegrees=35, twistLimitDegrees=15}
 {bone1Name=dyn_L_hairFront_3, bone2Name=dyn_L_hairFront_2, swing1LimitDegrees=35, swing2LimitDegrees=35, twistLimitDegrees=15}
 {
  bone1Name=dyn_R_hairFront_Lookat
  bone2Name=neck_01
  swing1LimitDegrees=45
  swing1Motion=Free
  swing2LimitDegrees=45
  swing2Motion=Free
  twistLimitDegrees=45
  twistMotion=Free
 }
 {
  bone1Name=dyn_R_hairFront_0
  bone2Name=dyn_R_hairFront_Lookat
  swing1LimitDegrees=45
  swing1Motion=Free
  swing2LimitDegrees=45
  swing2Motion=Free
  twistLimitDegrees=45
  twistMotion=Free
 }
 {bone1Name=dyn_R_hairFront_1, bone2Name=dyn_R_hairFront_0, swing1LimitDegrees=15, swing2LimitDegrees=15, twistLimitDegrees=5}
 {bone1Name=dyn_R_hairFront_2, bone2Name=dyn_R_hairFront_1, swing1LimitDegrees=25, swing2LimitDegrees=25, twistLimitDegrees=10}
 {bone1Name=dyn_R_hairFront_3, bone2Name=dyn_R_hairFront_2, swing1LimitDegrees=35, swing2LimitDegrees=35, twistLimitDegrees=15}
]
