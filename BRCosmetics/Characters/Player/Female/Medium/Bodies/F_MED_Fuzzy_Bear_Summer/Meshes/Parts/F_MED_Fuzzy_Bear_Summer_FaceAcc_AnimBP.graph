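# Animation Blueprint for the Fuzzy Bear Summer female face accessory, driving secondary motion on hair, ears and hood.
# It copies the body pose, simulates the drooped hood, then poses hair through constraints, look-ats and correctives for head, rifle-aim arm and one emote.
# The cached pose branches into front-end physics, in-game physics, or front physics with constrained back hair when back bling is worn.
# AnimDynamics finishes bangs and ears.
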AnimBlueprint /BRCosmetics/Characters/Player/Female/Medium/Bodies/F_MED_Fuzzy_Bear_Summer/Meshes/Parts/F_MED_Fuzzy_Bear_Summer_FaceAcc_AnimBP
target_skeleton=&/F_MED_Fuzzy_Bear_Summer_FaceAcc_Skeleton
graphs [
 AnimGraph {nodes [
   Component To Local()
   n1 = "Constraint - dyn_C_hairBack_0 (PP)"!(ComponentPose, Alpha=1, ConstraintWeights_0=0.5, ConstraintWeights_1:float=0.5) -> (Pose) {
    alphaBoolBlend {blendOption=Linear}
    boneToModify {boneName=dyn_C_hairBack_0}
    constraintSetup [{targetBone {boneName=dyn_L_hairBack_0}}, {targetBone {boneName=dyn_R_hairBack_0}}]
    constraintWeights [1,1]
   }
   n2 = "Constraint - dyn_C_hairBack_1 (PP)"!(ComponentPose=@n1, Alpha=1, ConstraintWeights_0=0.5, ConstraintWeights_1:float=0.5) -> (Pose) {
    alphaBoolBlend {blendOption=Linear}
    boneToModify {boneName=dyn_C_hairBack_1}
    constraintSetup [{targetBone {boneName=dyn_L_hairBack_1}}, {targetBone {boneName=dyn_R_hairBack_1}}]
    constraintWeights [1,1]
   }
   n3 = "Constraint - dyn_C_hairBack_2 (PP)"!(ComponentPose=@n2, Alpha=1, ConstraintWeights_0=0.5, ConstraintWeights_1:float=0.5) -> (Pose) {
    alphaBoolBlend {blendOption=Linear}
    boneToModify {boneName=dyn_C_hairBack_2}
    constraintSetup [{targetBone {boneName=dyn_L_hairBack_2}}, {targetBone {boneName=dyn_R_hairBack_2}}]
    constraintWeights [1,1]
   }
   "Constraint - dyn_C_hairBack_strand (PP)"!(ComponentPose=@n3, Alpha=1, ConstraintWeights_0=0.5, ConstraintWeights_1:float=0.5) -> (Pose) {
    alphaBoolBlend {blendOption=Linear}
    boneToModify {boneName=dyn_C_hairBack_strand}
    constraintSetup [{targetBone {boneName=dyn_L_hairBack_strand}}, {targetBone {boneName=dyn_R_hairBack_strand}}]
    constraintWeights [1,1]
   }
   n4 = "Constraint - dyn_R_hairBack_0 (PP)"!(ComponentPose, Alpha=1, ConstraintWeights_0=1, ConstraintWeights_1:float=0) -> (Pose) {
    alphaBoolBlend {blendOption=Linear}
    boneToModify {boneName=dyn_R_hairBack_0}
    constraintSetup [{targetBone {boneName=head}}, {targetBone {boneName=clavicle_r}}]
    constraintWeights [1,1]
   }
   n5 = "Constraint - dyn_C_hairBack_0 (P)"(ComponentPose=@n4, Alpha=1, ConstraintWeights_0=1) {
    alphaBoolBlend {blendOption=Linear}
    boneToModify {boneName=dyn_C_hairBack_0}
    constraintSetup [{targetBone {boneName=head}}]
    constraintWeights [1]
   }
   "Constraint - dyn_L_hairBack_0 (PP)"!(ComponentPose=@n5, Alpha=1, ConstraintWeights_0=1, ConstraintWeights_1:float=0) -> (Pose) {
    alphaBoolBlend {blendOption=Linear}
    boneToModify {boneName=dyn_L_hairBack_0}
    constraintSetup [{targetBone {boneName=head}}, {targetBone {boneName=clavicle_l}}]
    constraintWeights [1,1]
   }
   n6 = "Constraint - dyn_L_hairFront_Lookat (PP)"!(ComponentPose, Alpha=1, ConstraintWeights_0=0.6, ConstraintWeights_1:float=0.4) -> (Pose) {
    alphaBoolBlend {blendOption=Linear}
    boneToModify {boneName=dyn_L_hairFront_Lookat}
    constraintSetup [{targetBone {boneName=neck_01}}, {targetBone {boneName=head}}]
    constraintWeights [1,1]
   }
   "Constraint - dyn_R_hairFront_Lookat (PP)"!(ComponentPose=@n6, Alpha=1, ConstraintWeights_0=0.6, ConstraintWeights_1:float=0.4) -> (Pose) {
    alphaBoolBlend {blendOption=Linear}
    boneToModify {boneName=dyn_R_hairFront_Lookat}
    constraintSetup [{targetBone {boneName=neck_01}}, {targetBone {boneName=head}}]
    constraintWeights [1,1]
   }
   Local To Component()
   RigidBody!(ComponentPose, Alpha, OverrideWorldGravity:Vector=Get gravityFinal() -> (gravityFinal:Vector), ExternalForce) -> (Pose) {
    bClampLinearTranslationLimitToRefPose=true
    bOverrideWorldGravity=true
    overridePhysicsAsset=&/F_MED_Fuzzy_Bear_Summer_FaceAcc_Physics
    simulationSpace=BaseBoneSpace
    alphaBoolBlend {blendOption=Linear}
    baseBoneRef {boneName=neck_01}
    componentAppliedLinearAccClamp=<3000 3000 3000>
    componentLinearAccScale=<0.4 0.4 0.3>
    componentLinearVelScale=<0.4 0.4 0.3>
    overrideWorldGravity {z=-1111}
   }
   n7 = Anim Dynamics - Bone: dyn_bangs_1(ComponentPose=Local To Component(LocalPose=Use cached pose 'post sim'()), Alpha=1, ExternalForce="0,0,0") {
    angularDampingOverride=1
    angularSpringConstant=200
    bAngularSpring=true
    bOverrideAngularDamping=true
    bUseGravityOverride=true
    gravityScale=0.35
    alphaBoolBlend {blendOption=Linear}
    boundBone {boneName=dyn_bangs_1}
    componentLinearAccScale=<1 1 1>
    componentLinearVelScale=<1 1 1>
    gravityOverride {z=-1000}
    physicsBodyDefinitions [{
      boundBone {boneName=dyn_bangs_1}
      boxExtents=<10 10 10>
      constraintSetup {angularLimitsMax=<0 0 20>, angularLimitsMin=<0 0 -20>, angularTarget=<0 0 0>}
      localJointOffset=<2 0 0>
     }]
   }
   n8 = Anim Dynamics - Bone: dyn_bangs_2(ComponentPose=@n7, Alpha=1, ExternalForce="0,0,0") {
    angularDampingOverride=1
    angularSpringConstant=200
    bAngularSpring=true
    bOverrideAngularDamping=true
    bUseGravityOverride=true
    gravityScale=0.35
    alphaBoolBlend {blendOption=Linear}
    boundBone {boneName=dyn_bangs_2}
    componentLinearAccScale=<1 1 1>
    componentLinearVelScale=<1 1 1>
    gravityOverride {z=-1000}
    physicsBodyDefinitions [{
      boundBone {boneName=dyn_bangs_2}
      boxExtents=<10 10 10>
      constraintSetup {angularLimitsMax=<0 0 0>, angularLimitsMin=<0 -10 0>, angularTarget=<0 0 0>}
      localJointOffset=<2 0 0>
     }]
   }
   n9 = Anim Dynamics - Bone: dyn_L_earHoodDown(ComponentPose=@n8, Alpha=1, ExternalForce="0,0,0") {
    angularDampingOverride=1
    angularSpringConstant=20000
    bAngularSpring=true
    bOverrideAngularDamping=true
    bUseGravityOverride=true
    gravityScale=0.35
    alphaBoolBlend {blendOption=Linear}
    boundBone {boneName=dyn_L_earHoodDown}
    componentLinearAccScale=<1 1 1>
    componentLinearVelScale=<1 1 1>
    gravityOverride {z=-1000}
    physicsBodyDefinitions [{
      boundBone {boneName=dyn_L_earHoodDown}
      boxExtents=<20 20 20>
      constraintSetup {angularLimitsMax=<0 5 30>, angularLimitsMin=<0 -5 -30>, angularTarget=<0 0 0>}
      localJointOffset=<2 0 0>
     }]
   }
   n10 = Anim Dynamics - Bone: dyn_R_earHoodDown(ComponentPose=@n9, Alpha=1, ExternalForce="0,0,0") {
    angularDampingOverride=1
    angularSpringConstant=20000
    bAngularSpring=true
    bOverrideAngularDamping=true
    bUseGravityOverride=true
    gravityScale=0.35
    alphaBoolBlend {blendOption=Linear}
    boundBone {boneName=dyn_R_earHoodDown}
    componentLinearAccScale=<1 1 1>
    componentLinearVelScale=<1 1 1>
    gravityOverride {z=-1000}
    physicsBodyDefinitions [{
      boundBone {boneName=dyn_R_earHoodDown}
      boxExtents=<20 20 20>
      constraintSetup {angularLimitsMax=<0 5 30>, angularLimitsMin=<0 -5 -30>, angularTarget=<0 0 0>}
      localJointOffset=<-2 0 0>
     }]
   }
   Output Pose(Result=Component To Local(ComponentPose=@n10)) {layerGroup=None}
   n11 = Copy Pose From Mesh(SourceMeshComponent=Get Body_Mesh())
   n12 = Copy Pose From Mesh() {bCopyCurves=true, bUseAttachedParent=true}
   n13 = Anim Dynamics - Bone: dyn_droopedHood!(
    ComponentPose=Local To Component(LocalPose=Layered blend per bone(BasePose=@n12, BlendPoses_0=@n11) {layerSetup [{branchFilters [{boneName=upperarm_l}, {boneName=upperarm_r}]}]})
    bAlphaBoolEnabled:bool=NOT Boolean(A=Get backpack() -> (backpack:bool))
    ExternalForce="0,0,0"
   ) -> (Pose) {
    alphaInputType=Bool
    angularDampingOverride=0.98
    angularSpringConstant=100
    bGravityOverrideInSimSpace=true
    bOverrideAngularDamping=true
    simulationSpace=RootRelative
    alphaBoolBlend {blendInTime=0.1, blendOption=Linear, blendOutTime=0.1}
    boundBone {boneName=dyn_droopedHood}
    componentAppliedLinearAccClamp=<3500 3500 3500>
    componentLinearAccScale=<0.6 0.6 1>
    componentLinearVelScale=<0.6 0.6 1>
    physicsBodyDefinitions [{
      sphereCollisionRadius=0
      boundBone {boneName=dyn_droopedHood}
      boxExtents=<20 10 20>
      constraintSetup {angularLimitsMax=<1 5 5>, angularLimitsMin=<-1 -5 -45>, angularTarget=<1 0 0>}
      localJointOffset=<1 0 0>
     }]
    relativeSpaceBone {boneName=root}
   }
   n14 = "Constraint - dyn_Front_r_Hair_TGT (PP)"!(ComponentPose=@n13, Alpha=1, ConstraintWeights_0=0.24, ConstraintWeights_1:float=0.76) -> (Pose) {
    alphaBoolBlend {blendOption=Linear}
    boneToModify {boneName=dyn_Front_r_Hair_TGT}
    constraintSetup [{targetBone {boneName=spine_05}}, {targetBone {boneName=spine_04}}]
    constraintWeights [1,1]
   }
   n15 = "Constraint - dyn_Front_l_Hair_TGT (PP)"!(ComponentPose=@n14, Alpha=1, ConstraintWeights_0=0.24, ConstraintWeights_1:float=0.76) -> (Pose) {
    alphaBoolBlend {blendOption=Linear}
    boneToModify {boneName=dyn_Front_l_Hair_TGT}
    constraintSetup [{targetBone {boneName=spine_05}}, {targetBone {boneName=spine_04}}]
    constraintWeights [1,1]
   }
   n16 = "Constraint - dyn_hair_TGT (PPPP)"!(
    ComponentPose=@n15
    Alpha=1
    ConstraintWeights_0=0.0791
    ConstraintWeights_1:float=0.1117
    ConstraintWeights_2:float=0.7433
    ConstraintWeights_3:float=0.0659
   ) -> (Pose) {
    alphaBoolBlend {blendOption=Linear}
    boneToModify {boneName=dyn_hair_TGT}
    constraintSetup [
     {targetBone {boneName=clavicle_l}}
     {targetBone {boneName=spine_05}}
     {targetBone {boneName=dyn_droopedHood}}
     {targetBone {boneName=clavicle_r}}
    ]
    constraintWeights [1,1,1,1]
   }
   n17 = "Constraint - dyn_l_hair_TGT (PPPPPPPP)"!(
    ComponentPose=@n16
    Alpha=1
    ConstraintWeights_0=0.0054
    ConstraintWeights_1:float=0.1021
    ConstraintWeights_2:float=0.0098
    ConstraintWeights_3:float=0.077
    ConstraintWeights_4:float=0.7335
    ConstraintWeights_5:float=0.0122
    ConstraintWeights_6:float=0.0381
    ConstraintWeights_7:float=0.0219
   ) -> (Pose) {
    alphaBoolBlend {blendOption=Linear}
    boneToModify {boneName=dyn_l_hair_TGT}
    constraintSetup [
     {targetBone {boneName=deform_upperarm_up_l}}
     {targetBone {boneName=clavicle_l}}
     {targetBone {boneName=deform_upperarm_bk_l}}
     {targetBone {boneName=spine_05}}
     {targetBone {boneName=dyn_droopedHood}}
     {targetBone {boneName=upperarm_twist_01_l}}
     {targetBone {boneName=dyn_L_earHoodDown}}
     {targetBone {boneName=clavicle_r}}
    ]
    constraintWeights [1,1,1,1,1,1,1,1]
   }
   n18 = "Constraint - dyn_r_hair_TGT (PPPPPPP)"!(
    ComponentPose=@n17
    Alpha=1
    ConstraintWeights_0=0.0109
    ConstraintWeights_1:float=0.0241
    ConstraintWeights_2:float=0.0634
    ConstraintWeights_3:float=0.7648
    ConstraintWeights_4:float=0.0355
    ConstraintWeights_5:float=0.0917
    ConstraintWeights_6:float=0.0096
   ) -> (Pose) {
    alphaBoolBlend {blendOption=Linear}
    boneToModify {boneName=dyn_r_hair_TGT}
    constraintSetup [
     {targetBone {boneName=upperarm_twist_01_r}}
     {targetBone {boneName=clavicle_l}}
     {targetBone {boneName=spine_05}}
     {targetBone {boneName=dyn_droopedHood}}
     {targetBone {boneName=dyn_R_earHoodDown}}
     {targetBone {boneName=clavicle_r}}
     {targetBone {boneName=deform_upperarm_bk_r}}
    ]
    constraintWeights [1,1,1,1,1,1,1]
   }
   n19 = "Constraint - dyn_L_hairFront_0 (PP)"!(ComponentPose=@n18, Alpha=1, ConstraintWeights_0=0.4, ConstraintWeights_1:float=0.6) -> (Pose) {
    alphaBoolBlend {blendOption=Linear}
    boneToModify {boneName=dyn_L_hairFront_0}
    constraintSetup [{targetBone {boneName=clavicle_l}}, {targetBone {boneName=spine_05}}]
    constraintWeights [1,1]
   }
   n20 = "Constraint - dyn_R_hairFront_Lookat (PP)"(ComponentPose=@n19, Alpha=1, ConstraintWeights_0=0.3, ConstraintWeights_1=0.7) {
    alphaBoolBlend {blendOption=Linear}
    boneToModify {boneName=dyn_R_hairFront_Lookat}
    constraintSetup [{targetBone {boneName=clavicle_r}}, {targetBone {boneName=spine_05}}]
    constraintWeights [1,1]
   }
   n21 = Look At - Bone: dyn_C_hairBack_Lookat(ComponentPose=@n20, Alpha=1) {
    alphaBoolBlend {blendOption=Linear}
    boneToModify {boneName=dyn_C_hairBack_Lookat}
    lookAtLocation {x=0}
    lookAtTarget {boneReference {boneName=dyn_hair_TGT}}
    lookAt_Axis {axis=<1 0>}
   }
   n22 = Look At - Bone: dyn_R_hairBack_Lookat(ComponentPose=@n21, Alpha=1) {
    alphaBoolBlend {blendOption=Linear}
    boneToModify {boneName=dyn_R_hairBack_Lookat}
    lookAtLocation {x=0}
    lookAtTarget {boneReference {boneName=dyn_r_hair_TGT}}
    lookAt_Axis {axis=<1 0>}
   }
   n23 = Look At - Bone: dyn_L_hairBack_Lookat(ComponentPose=@n22, Alpha=1) {
    alphaBoolBlend {blendOption=Linear}
    boneToModify {boneName=dyn_L_hairBack_Lookat}
    lookAtLocation {x=0}
    lookAtTarget {boneReference {boneName=dyn_l_hair_TGT}}
    lookAt_Axis {axis=<1 0>}
   }
   n24 = Look At - Bone: dyn_R_hairFront_Lookat(ComponentPose=@n23, Alpha=1) {
    alphaBoolBlend {blendOption=Linear}
    boneToModify {boneName=dyn_R_hairFront_Lookat}
    lookAtLocation=<4.5448 3.7001>
    lookAtTarget {boneReference {boneName=dyn_Front_r_Hair_TGT}}
    lookAt_Axis {axis=<1 0>}
   }
   n25 = Get head rot() -> (head rot:double)
   n26 = "Transform (Modify) Bone - Bone: dyn_R_hairBack_Lookat"(ComponentPose=@n24, Alpha=@n25, Translation="0,-3,0", Rotation="0,-15,0", Scale="1,1,1") {
    rotationMode=BMM_Additive
    rotationSpace=BCS_BoneSpace
    translationMode=BMM_Additive
    translationSpace=BCS_BoneSpace
    alphaBoolBlend {blendOption=Linear}
    boneToModify {boneName=dyn_R_hairBack_Lookat}
    rotation {yaw=-15}
    translation {y=-3}
   }
   n27 = "Transform (Modify) Bone - Bone: dyn_C_hairBack_Lookat"(ComponentPose=@n26, Alpha=@n25, Translation="0,-3,0", Rotation="0,-15,0", Scale="1,1,1") {
    rotationMode=BMM_Additive
    rotationSpace=BCS_BoneSpace
    translationMode=BMM_Additive
    translationSpace=BCS_BoneSpace
    alphaBoolBlend {blendOption=Linear}
    boneToModify {boneName=dyn_C_hairBack_Lookat}
    rotation {yaw=-15}
    translation {y=-3}
   }
   n28 = "Transform (Modify) Bone - Bone: dyn_L_hairBack_Lookat"(ComponentPose=@n27, Alpha=@n25, Translation="0,-3,0", Rotation="0,-15,0", Scale="1,1,1") {
    rotationMode=BMM_Additive
    rotationSpace=BCS_BoneSpace
    translationMode=BMM_Additive
    translationSpace=BCS_BoneSpace
    alphaBoolBlend {blendOption=Linear}
    boneToModify {boneName=dyn_L_hairBack_Lookat}
    rotation {yaw=-15}
    translation {y=-3}
   }
   n29 = Get Arm L rot() -> (Arm L rot:double)
   n30 = "Transform (Modify) Bone - Bone: dyn_C_hairBack_Lookat"(ComponentPose=@n28, Alpha=@n29, Translation="0,-3,0", Rotation="0,0,0", Scale="1,1,1") {
    rotationMode=BMM_Additive
    rotationSpace=BCS_BoneSpace
    translationMode=BMM_Additive
    translationSpace=BCS_BoneSpace
    alphaBoolBlend {blendInTime=0.1, blendOption=Linear, blendOutTime=0.1}
    alphaScaleBiasClamp {bInterpResult=true}
    boneToModify {boneName=dyn_C_hairBack_Lookat}
    translation {y=-3}
   }
   n31 = "Transform (Modify) Bone - Bone: dyn_C_hairBack_Lookat"(ComponentPose=@n30, Alpha=@n29, Translation="0,-4,0", Rotation="0,0,0", Scale="1,1,1") {
    rotationMode=BMM_Additive
    rotationSpace=BCS_BoneSpace
    translationMode=BMM_Additive
    translationSpace=BCS_BoneSpace
    alphaBoolBlend {blendInTime=0.1, blendOption=Linear, blendOutTime=0.1}
    alphaScaleBiasClamp {bInterpResult=true}
    boneToModify {boneName=dyn_C_hairBack_Lookat}
    translation {y=-4}
   }
   n32 = "Transform (Modify) Bone - Bone: dyn_R_hairBack_Lookat"(ComponentPose=@n31, Alpha=@n29, Translation="0,-5,-1", Rotation="0,0,0", Scale="1,1,1") {
    rotationMode=BMM_Additive
    rotationSpace=BCS_BoneSpace
    translationMode=BMM_Additive
    translationSpace=BCS_BoneSpace
    alphaBoolBlend {blendInTime=0.1, blendOption=Linear, blendOutTime=0.1}
    alphaScaleBiasClamp {bInterpResult=true}
    boneToModify {boneName=dyn_R_hairBack_Lookat}
    translation {y=-5, z=-1}
   }
   n33 = "Transform (Modify) Bone - Bone: dyn_L_hairFront_0"(
    ComponentPose=@n32
    Alpha=Get isCustomEmotePlaying() -> (isCustomEmotePlaying:double)
    Translation="0.000,-3.361,3.133"
    Rotation="0.000,-0.000,-1.654"
    Scale="1,1,1"
   ) {
    rotationMode=BMM_Replace
    rotationSpace=BCS_BoneSpace
    scaleMode=BMM_Replace
    scaleSpace=BCS_BoneSpace
    translationMode=BMM_Replace
    translationSpace=BCS_BoneSpace
    alphaBoolBlend {blendOption=Linear}
    boneToModify {boneName=dyn_L_hairFront_0}
    rotation=<rot 0 0 -1.654>
    translation {y=-3.361, z=3.133}
   }
   n34 = "Transform (Modify) Bone - Bone: dyn_L_hairFront_0"!(
    ComponentPose=@n33
    bAlphaBoolEnabled:bool=Get isNotPlayingCustomEmote() -> (isNotPlayingCustomEmote:bool)
    Translation="0,0,0.7"
    Rotation="0,0,0"
    Scale="1,1,1"
   ) -> (Pose) {
    alphaInputType=Bool
    rotationMode=BMM_Replace
    rotationSpace=BCS_BoneSpace
    scaleMode=BMM_Replace
    scaleSpace=BCS_BoneSpace
    translationMode=BMM_Replace
    translationSpace=BCS_BoneSpace
    alphaBoolBlend {blendInTime=0.3, blendOption=Linear, blendOutTime=0.3}
    boneToModify {boneName=dyn_L_hairFront_0}
    translation {z=0.7}
   }
   Save cached pose 'presim'(Pose=Component To Local(ComponentPose=@n34)) {cacheName=presim}
   n35 = RigidBody(ComponentPose=Local To Component(LocalPose=Use cached pose 'presim'()), OverrideWorldGravity="0,0,-980") {
    bClampLinearTranslationLimitToRefPose=true
    bOverrideWorldGravity=true
    overridePhysicsAsset=&/F_MED_Fuzzy_Bear_Summer_Hood_Up_Front_FaceAcc_Physics
    simulationSpace=BaseBoneSpace
    alphaBoolBlend {blendOption=Linear}
    baseBoneRef {boneName=spine_02}
    componentAppliedLinearAccClamp=<3500 3500 3500>
    componentLinearAccScale=<0.3 0.3 0.4>
    componentLinearVelScale=<0.3 0.3 0.4>
   }
   n36 = RigidBody!(
    ComponentPose=@n35
    Alpha
    OverrideWorldGravity=Get General Gravity Override() -> (General Gravity Override:Vector)
    ExternalForce
    ComponentLinearAccScale:Vector=Get AccScale() -> (AccScale:Vector)
    ComponentLinearVelScale:Vector=Get VelScale() -> (VelScale:Vector)
   ) -> (Pose) {
    bClampLinearTranslationLimitToRefPose=true
    bOverrideWorldGravity=true
    overridePhysicsAsset=&/F_MED_Fuzzy_Bear_Summer_Back_FaceAcc_Physics
    simulationSpace=BaseBoneSpace
    alphaBoolBlend {blendOption=Linear}
    baseBoneRef {boneName=spine_05}
    componentAppliedLinearAccClamp=<3000 3000 3000>
    componentLinearAccScale=<0.4 0.3 0.4>
    componentLinearVelScale=<0.4 0.3 0.4>
    overrideWorldGravity {z=-1111}
   }
   n37 = Component To Local(ComponentPose=@n36)
   n38 = RigidBody!(ComponentPose=Local To Component(LocalPose=Use cached pose 'presim'()), Alpha, OverrideWorldGravity="0,0,-980", ExternalForce) -> (Pose) {
    bClampLinearTranslationLimitToRefPose=true
    bOverrideWorldGravity=true
    overridePhysicsAsset=&F_MED_Fuzzy_Bear_Summer_Hood_Up_Front_FaceAcc_Physics
    simulationSpace=BaseBoneSpace
    alphaBoolBlend {blendOption=Linear}
    baseBoneRef {boneName=spine_02}
    componentAppliedLinearAccClamp=<3500 3500 3500>
    componentLinearAccScale=<0.3 0.3 0.4>
    componentLinearVelScale=<0.3 0.3 0.4>
   }
   n39 = "Constraint - dyn_R_hairBack_1 (PPP)"!(ComponentPose=@n38, Alpha=1, ConstraintWeights_0=0.15, ConstraintWeights_1:float=0.1, ConstraintWeights_2:float=0.75) -> (Pose) {
    alphaBoolBlend {blendOption=Linear}
    boneToModify {boneName=dyn_R_hairBack_1}
    constraintSetup [{targetBone {boneName=spine_05}}, {targetBone {boneName=neck_01}}, {targetBone {boneName=clavicle_r}}]
    constraintWeights [1,1,1]
   }
   n40 = "Constraint - dyn_C_hairBack_1 (PP)"(ComponentPose=@n39, Alpha=1, ConstraintWeights_0=0.5, ConstraintWeights_1=0.5) {
    alphaBoolBlend {blendOption=Linear}
    boneToModify {boneName=dyn_C_hairBack_1}
    constraintSetup [{targetBone {boneName=spine_05}}, {targetBone {boneName=neck_01}}]
    constraintWeights [1,1]
   }
   n41 = "Constraint - dyn_L_hairBack_1 (PPP)"!(ComponentPose=@n40, Alpha=1, ConstraintWeights_0=0.15, ConstraintWeights_1:float=0.1, ConstraintWeights_2:float=0.75) -> (Pose) {
    alphaBoolBlend {blendOption=Linear}
    boneToModify {boneName=dyn_L_hairBack_1}
    constraintSetup [{targetBone {boneName=spine_05}}, {targetBone {boneName=neck_01}}, {targetBone {boneName=clavicle_l}}]
    constraintWeights [1,1,1]
   }
   n42 = "Constraint - dyn_R_hairBack_2 (PPP)"!(ComponentPose=@n41, Alpha=1, ConstraintWeights_0=0, ConstraintWeights_1:float=0.25, ConstraintWeights_2:float=0.75) -> (Pose) {
    alphaBoolBlend {blendOption=Linear}
    boneToModify {boneName=dyn_R_hairBack_2}
    constraintSetup [{targetBone {boneName=neck_01}}, {targetBone {boneName=spine_05}}, {targetBone {boneName=clavicle_r}}]
    constraintWeights [1,1,1]
   }
   n43 = "Constraint - dyn_C_hairBack_2 (PP)"(ComponentPose=@n42, Alpha=1, ConstraintWeights_0=0.5, ConstraintWeights_1=0.5) {
    alphaBoolBlend {blendOption=Linear}
    boneToModify {boneName=dyn_C_hairBack_2}
    constraintSetup [{targetBone {boneName=neck_01}}, {targetBone {boneName=spine_05}}]
    constraintWeights [1,1]
   }
   n44 = "Constraint - dyn_L_hairBack_2 (PPP)"!(ComponentPose=@n43, Alpha=1, ConstraintWeights_0=0, ConstraintWeights_1:float=0.25, ConstraintWeights_2:float=0.75) -> (Pose) {
    alphaBoolBlend {blendOption=Linear}
    boneToModify {boneName=dyn_L_hairBack_2}
    constraintSetup [{targetBone {boneName=neck_01}}, {targetBone {boneName=spine_05}}, {targetBone {boneName=clavicle_l}}]
    constraintWeights [1,1,1]
   }
   n45 = "Constraint - dyn_R_hairBack_strand (PP)"!(ComponentPose=@n44, Alpha=1, ConstraintWeights_0=0.5, ConstraintWeights_1:float=0.5) -> (Pose) {
    alphaBoolBlend {blendOption=Linear}
    boneToModify {boneName=dyn_R_hairBack_strand}
    constraintSetup [{targetBone {boneName=spine_05}}, {targetBone {boneName=spine_04}}]
    constraintWeights [1,1]
   }
   n46 = "Constraint - dyn_C_hairBack_strand (PP)"(ComponentPose=@n45, Alpha=1, ConstraintWeights_0=0.5, ConstraintWeights_1=0.5) {
    alphaBoolBlend {blendOption=Linear}
    boneToModify {boneName=dyn_C_hairBack_strand}
    constraintSetup [{targetBone {boneName=spine_05}}, {targetBone {boneName=spine_04}}]
    constraintWeights [1,1]
   }
   n47 = "Constraint - dyn_L_hairBack_strand (PP)"!(ComponentPose=@n46, Alpha=1, ConstraintWeights_0=0.5, ConstraintWeights_1:float=0.5) -> (Pose) {
    alphaBoolBlend {blendOption=Linear}
    boneToModify {boneName=dyn_L_hairBack_strand}
    constraintSetup [{targetBone {boneName=spine_05}}, {targetBone {boneName=spine_04}}]
    constraintWeights [1,1]
   }
   n48 = Component To Local(ComponentPose=@n47)
   n49 = Get backpack()
   n50 = RigidBody(ComponentPose=Local To Component(LocalPose=Use cached pose 'presim'()), OverrideWorldGravity="0,0,-980") {
    bClampLinearTranslationLimitToRefPose=true
    bOverrideWorldGravity=true
    overridePhysicsAsset=&/F_MED_Fuzzy_Bear_Summer_Hood_Up_FrontEnd_FaceAcc_Physics
    simulationSpace=BaseBoneSpace
    alphaBoolBlend {blendOption=Linear}
    baseBoneRef {boneName=spine_02}
    componentAppliedLinearAccClamp=<3500 3500 3500>
    componentLinearAccScale=<0.3 0.3 0.4>
    componentLinearVelScale=<0.3 0.3 0.4>
   }
   n51 = RigidBody!(
    ComponentPose=@n50
    Alpha
    OverrideWorldGravity=Get General Gravity Override()
    ExternalForce
    ComponentLinearAccScale:Vector=Get AccScale()
    ComponentLinearVelScale:Vector=Get VelScale()
   ) -> (Pose) {
    bClampLinearTranslationLimitToRefPose=true
    bOverrideWorldGravity=true
    overridePhysicsAsset=&/F_MED_Fuzzy_Bear_Summer_Back_FaceAcc_Frontend_Physics
    simulationSpace=BaseBoneSpace
    alphaBoolBlend {blendOption=Linear}
    baseBoneRef {boneName=spine_05}
    componentAppliedLinearAccClamp=<3000 3000 3000>
    componentLinearAccScale=<0.4 0.3 0.4>
    componentLinearVelScale=<0.4 0.3 0.4>
    overrideWorldGravity {z=-1111}
   }
   Save cached pose 'post sim'(Pose=Blend Poses by bool(
     BlendPose_0=Component To Local(ComponentPose=@n51)
     BlendPose_1=Blend Poses by bool(BlendPose_0=@n48, BlendPose_1=@n37, bActiveValue=@n49) {blendType=Linear}
     bActiveValue=Get bIsInFrontEnd()
    ) {blendType=Linear}) {cacheName=post sim}
   Get gravityFinal()
  ]}
 EventGraph {nodes [
   Event BlueprintUpdateAnimation()
   > n1 = Sequence!(execute) -> (then_0, then_1, then_2:Exec, then_3:Exec)
   > n2 = Gravity Override State Function DEPRECATED(CharacterPartAnimInstance=Self-Reference(), GameTimeInSeconds=GetGameTimeInSeconds(), GravityOverrideParams=Get Gravity Override Params())
   > Set General Gravity Override(General Gravity Override:Vector=@n2) -> (Output_Get:Vector)
   n3 = CheckForBackbling()
   n4 = AND Boolean(A=@n3, B=NOT Boolean(A=Get bIsInFrontEnd()))
   > Set backblingEquippedGravity(backblingEquippedGravity:Vector=SelectVector(A="0,0,-3000", B="0, 0,-1111", bPickA=@n4)) -> (Output_Get:Vector)
   > Set backpack(backpack:bool=@n3) -> (Output_Get:bool)
   > Set gravityFinal(gravityFinal:Vector=SelectVector(A=Get backblingEquippedGravity() -> (backblingEquippedGravity:Vector), B=Get General Gravity Override(), bPickA=@n4)) -> (Output_Get:Vector)
   > Set head rot(head rot:double=GetAlphaFromReferenceTransformDelta(SkeletalMeshComponent=Get Skel_Mesh(), BoneName=head, BaseName=spine_05, InputRange=(X=-20,Y=-50), TwistAxis="1,0,0")) -> (Output_Get:double)
   n5 = Cast To FortPlayerAnimInstance(execute=@n1.then_1, Object=GetAnimInstance(self=GetOwnerSkeletalMeshComponent()))
   > Set Rifle(Rifle:bool=false) -> (Output_Get:bool)
   > n6 = Switch on EFortWeaponCoreAnimation!(execute, Selection:EFortWeaponCoreAnimation Enum=Get WeaponCoreAnim(self=@n5)) -> (
    Melee:Exec
    Pistol:Exec
    Shotgun:Exec
    PaperBlueprint:Exec
    Rifle:Exec
    MeleeOneHand:Exec
    MachinePistol:Exec
    RocketLauncher:Exec
    GrenadeLauncher:Exec
    GoingCommando:Exec
    AssaultRifle:Exec
    TacticalShotgun:Exec
    SniperRifle:Exec
    TrapPlacement:Exec
    ShoulderLauncher:Exec
    AbilityDecoTool:Exec
    Crossbow:Exec
    C4:Exec
    RemoteControl:Exec
    DualWield:Exec
    AR_BullPup:Exec
    AR_ForwardGrip:Exec
    MedPackPaddles:Exec
    SMG_P90:Exec
    AR_DrumGun:Exec
    Consumable_Small:Exec
    Consumable_Large:Exec
    Balloon:Exec
    MountedTurret:Exec
    CreativeTool:Exec
    ExplosiveBow:Exec
    AshtonIndigo:Exec
    AshtonChicago:Exec
    MeleeDualWield:Exec
    Unarmed:Exec
   )
   n7 = Branch(
    execute=[@n6.AssaultRifle, @n6.Crossbow, @n6.GrenadeLauncher, @n6.Rifle, @n6.RocketLauncher, @n6.Shotgun, @n6.SniperRifle, @n6.TacticalShotgun]
    Condition=Get bIsTargeting()
   )
   > Set Rifle(Rifle=true)
   > Set Arm L rot(Arm L rot:double=GetAlphaFromReferenceTransformDelta(
     SkeletalMeshComponent=Get Skel_Mesh()
     BoneName=upperarm_l
     BaseName=spine_05
     DeltaComponentType=EulerY
     InputRange=(X=15,Y=60)
     TwistAxis="1,0,0"
    )) -> (Output_Get:double)
   Set Arm L rot(execute=@n7.else, Arm L rot=0)
   n8 = Branch(execute=@n1.then_2, Condition=Get backpack())
   > Set AccScale(AccScale:Vector="0, 0, 0") -> (Output_Get:Vector)
   > Set VelScale(VelScale:Vector="0, 0, 0") -> (Output_Get:Vector)
   Set AccScale(execute=@n8.else, AccScale="0.4,0.3,0.4")
   > Set VelScale(VelScale="0.4,0.3,0.4")
   n9 = GetAnimInstance(self=GetOwnerSkeletalMeshComponent())
   Is Valid(exec=@n1.then_3, InputObject=@n9)
   n10 = Montage_IsPlaying(self=@n9, Montage=/BRCosmetics/Animation/Game/MainPlayer/Emotes/Emote_Celebration_Female_M)
   > Set isCustomEmotePlaying(isCustomEmotePlaying:double=SelectFloat(A=1, B=0, bPickA=@n10)) -> (Output_Get:double)
   > Set isNotPlayingCustomEmote(isNotPlayingCustomEmote:bool=NOT Boolean(A=@n10)) -> (Output_Get:bool)
   Event BlueprintInitializeAnimation()
   n11 = GetOwnerSkeletalMeshComponent()
   > Set Body_Mesh(Body_Mesh=@n11)
   > Set Skel_Mesh(Skel_Mesh=@n11)
  ]}
]
variables [
 AccScale {type=Vector}
 Arm L rot {type=number}
 backblingEquippedGravity {type=Vector}
 backpack {type=boolean}
 Body_Mesh {type=SkeletalMeshComponent}
 General Gravity Override {type=Vector}
 Gravity Override Params {type=GravityOverrideParamsStruct}
 gravityFinal {type=Vector}
 head rot {type=number}
 isCustomEmotePlaying {type=number}
 isNotPlayingCustomEmote {type=boolean}
 Rifle {type=boolean}
 Skel_Mesh {type=SkeletalMeshComponent}
 VelScale {type=Vector}
]
