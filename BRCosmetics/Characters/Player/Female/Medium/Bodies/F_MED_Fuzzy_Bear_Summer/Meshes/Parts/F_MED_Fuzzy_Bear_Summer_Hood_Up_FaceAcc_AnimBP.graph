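# Post-process style AnimBP for the female Fuzzy Bear Summer hood-up face accessory, adding secondary motion to hair, ears, bangs and hood.
# Copied poses get hair constraints and a look-at, then one of two RigidBody sims chosen by bIsInFrontEnd.
# Hood constraints between head and spine_05 follow, then chained AnimDynamics for ears, bangs and hood.
# The event graph computes backbling-aware gravity and head rotation, neither visibly read by the AnimGraph.
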AnimBlueprint /BRCosmetics/Characters/Player/Female/Medium/Bodies/F_MED_Fuzzy_Bear_Summer/Meshes/Parts/F_MED_Fuzzy_Bear_Summer_Hood_Up_FaceAcc_AnimBP
target_skeleton=&/F_MED_Fuzzy_Bear_Summer_Hood_Up_FaceAcc_Skeleton
graphs [
 AnimGraph {nodes [
   RigidBody() {alphaBoolBlend {blendOption=Linear}}
   n1 = "Constraint - dyn_R_hood (PP)"!(ComponentPose=Local To Component(LocalPose=Use cached pose 'post sim'()), Alpha=1, ConstraintWeights_0=0.8, ConstraintWeights_1:float=0.2) -> (Pose) {
    alphaBoolBlend {blendOption=Linear}
    boneToModify {boneName=dyn_R_hood}
    constraintSetup [{targetBone {boneName=head}}, {targetBone {boneName=spine_05}}]
    constraintWeights [1,1]
   }
   n2 = "Constraint - dyn_L_hood (PP)"!(ComponentPose=@n1, Alpha=1, ConstraintWeights_0=0.9, ConstraintWeights_1:float=0.1) -> (Pose) {
    alphaBoolBlend {blendOption=Linear}
    boneToModify {boneName=dyn_L_hood}
    constraintSetup [{targetBone {boneName=head}}, {targetBone {boneName=spine_05}}]
    constraintWeights [1,1]
   }
   n3 = Anim Dynamics - Bone: dyn_L_ear(ComponentPose=@n2, Alpha=1, ExternalForce="0,0,0") {
    angularDampingOverride=1
    angularSpringConstant=20000
    bAngularSpring=true
    bOverrideAngularDamping=true
    bUseGravityOverride=true
    gravityScale=0.35
    alphaBoolBlend {blendOption=Linear}
    boundBone {boneName=dyn_L_ear}
    componentLinearAccScale=<1 1 1>
    componentLinearVelScale=<1 1 1>
    gravityOverride {z=-1000}
    physicsBodyDefinitions [{
      boundBone {boneName=dyn_L_ear}
      boxExtents=<21 21 21>
      constraintSetup {angularLimitsMax=<0 25 -30>, angularLimitsMin=<0 -25 -30>, angularTarget=<0 0 0>}
      localJointOffset=<3 0 0>
     }]
   }
   n4 = Anim Dynamics - Bone: dyn_R_ear(ComponentPose=@n3, Alpha=1, ExternalForce="0,0,0") {
    angularDampingOverride=1
    angularSpringConstant=20000
    bAngularSpring=true
    bOverrideAngularDamping=true
    bUseGravityOverride=true
    gravityScale=0.35
    alphaBoolBlend {blendOption=Linear}
    boundBone {boneName=dyn_R_ear}
    componentLinearAccScale=<1 1 1>
    componentLinearVelScale=<1 1 1>
    gravityOverride {z=-1000}
    physicsBodyDefinitions [{
      boundBone {boneName=dyn_R_ear}
      boxExtents=<21 21 21>
      constraintSetup {angularLimitsMax=<0 25 -30>, angularLimitsMin=<0 -25 -30>, angularTarget=<0 0 0>}
      localJointOffset=<3 0 0>
     }]
   }
   n5 = Anim Dynamics - Bone: dyn_bangs_1(ComponentPose=@n4, Alpha=1, ExternalForce="0,0,0") {
    angularDampingOverride=1
    angularSpringConstant=200
    bAngularSpring=true
    bOverrideAngularDamping=true
    bUseGravityOverride=true
    gravityScale=0.35
    alphaBoolBlend {blendOption=Linear}
    boundBone {boneName=dyn_bangs_1}
    componentLinearAccScale=<1 1 1>
    componentLinearVelScale=<1 1 1>
    gravityOverride {z=-1000}
    physicsBodyDefinitions [{
      boundBone {boneName=dyn_bangs_1}
      boxExtents=<10 10 10>
      constraintSetup {angularLimitsMax=<0 0 20>, angularLimitsMin=<0 0 -20>, angularTarget=<0 0 0>}
      localJointOffset=<2 0 0>
     }]
   }
   n6 = Anim Dynamics - Bone: dyn_bangs_2(ComponentPose=@n5, Alpha=1, ExternalForce="0,0,0") {
    angularDampingOverride=1
    angularSpringConstant=200
    bAngularSpring=true
    bOverrideAngularDamping=true
    bUseGravityOverride=true
    gravityScale=0.35
    alphaBoolBlend {blendOption=Linear}
    boundBone {boneName=dyn_bangs_2}
    componentLinearAccScale=<1 1 1>
    componentLinearVelScale=<1 1 1>
    gravityOverride {z=-1000}
    physicsBodyDefinitions [{
      boundBone {boneName=dyn_bangs_2}
      boxExtents=<10 10 10>
      constraintSetup {angularLimitsMax=<0 10 5>, angularLimitsMin=<0 -10 -5>, angularTarget=<0 0 0>}
      localJointOffset=<2 0 0>
     }]
   }
   n7 = Anim Dynamics - Bone: dyn_L_hood(ComponentPose=@n6, Alpha=1, ExternalForce="0,0,0") {
    angularDampingOverride=1
    angularSpringConstant=20000
    bLinearSpring=true
    bOverrideLinearDamping=true
    bUseGravityOverride=true
    gravityScale=0.35
    linearDampingOverride=1
    linearSpringConstant=20
    simulationSpace=RootRelative
    alphaBoolBlend {blendOption=Linear}
    boundBone {boneName=dyn_L_hood}
    componentLinearAccScale=<1 1 1>
    componentLinearVelScale=<1 1 1>
    gravityOverride {z=-2600}
    physicsBodyDefinitions [{
      boundBone {boneName=dyn_L_hood}
      boxExtents=<70 70 70>
      constraintSetup {angularLimitsMax=<0 0 0>, angularLimitsMin=<0 0 0>, angularTarget=<0 0 0>, linearAxesMax=<2 0 3>}
      localJointOffset=<0 0 0>
     }]
   }
   n8 = Anim Dynamics - Bone: dyn_R_hood(ComponentPose=@n7, Alpha=1, ExternalForce="0,0,0") {
    angularDampingOverride=1
    angularSpringConstant=20000
    bLinearSpring=true
    bOverrideLinearDamping=true
    bUseGravityOverride=true
    gravityScale=0.35
    linearDampingOverride=1
    linearSpringConstant=20
    simulationSpace=RootRelative
    alphaBoolBlend {blendOption=Linear}
    boundBone {boneName=dyn_R_hood}
    componentLinearAccScale=<1 1 1>
    componentLinearVelScale=<1 1 1>
    gravityOverride {z=-2600}
    physicsBodyDefinitions [{
      boundBone {boneName=dyn_R_hood}
      boxExtents=<70 70 70>
      constraintSetup {angularLimitsMax=<0 0 0>, angularLimitsMin=<0 0 0>, angularTarget=<0 0 0>, linearAxesMin=<-1 0 -7>}
      localJointOffset=<0 0 0>
     }]
   }
   Output Pose(Result=Component To Local(ComponentPose=@n8)) {layerGroup=None}
   n9 = Copy Pose From Mesh(SourceMeshComponent=Get Body_Mesh())
   n10 = Copy Pose From Mesh() {bUseAttachedParent=true}
   n11 = "Constraint - dyn_Front_r_Hair_TGT (PP)"!(
    ComponentPose=Local To Component(LocalPose=Layered blend per bone(BasePose=@n10, BlendPoses_0=@n9) {layerSetup [{branchFilters [{boneName=upperarm_l}, {boneName=upperarm_r}]}]})
    Alpha=1
    ConstraintWeights_0=0.24
    ConstraintWeights_1:float=0.76
   ) -> (Pose) {
    alphaBoolBlend {blendOption=Linear}
    boneToModify {boneName=dyn_Front_r_Hair_TGT}
    constraintSetup [{targetBone {boneName=spine_05}}, {targetBone {boneName=spine_04}}]
    constraintWeights [1,1]
   }
   n12 = "Constraint - dyn_Front_l_Hair_TGT (PP)"!(ComponentPose=@n11, Alpha=1, ConstraintWeights_0=0.24, ConstraintWeights_1:float=0.76) -> (Pose) {
    alphaBoolBlend {blendOption=Linear}
    boneToModify {boneName=dyn_Front_l_Hair_TGT}
    constraintSetup [{targetBone {boneName=spine_05}}, {targetBone {boneName=spine_04}}]
    constraintWeights [1,1]
   }
   n13 = "Constraint - dyn_L_hairFront_0 (PP)"!(ComponentPose=@n12, Alpha=1, ConstraintWeights_0=0.4, ConstraintWeights_1:float=0.6) -> (Pose) {
    alphaBoolBlend {blendOption=Linear}
    boneToModify {boneName=dyn_L_hairFront_0}
    constraintSetup [{targetBone {boneName=clavicle_l}}, {targetBone {boneName=spine_05}}]
    constraintWeights [1,1]
   }
   n14 = "Constraint - dyn_R_hairFront_Lookat (PP)"!(ComponentPose=@n13, Alpha=1, ConstraintWeights_0=0.3, ConstraintWeights_1:float=0.7) -> (Pose) {
    alphaBoolBlend {blendOption=Linear}
    boneToModify {boneName=dyn_R_hairFront_Lookat}
    constraintSetup [{targetBone {boneName=clavicle_r}}, {targetBone {boneName=spine_05}}]
    constraintWeights [1,1]
   }
   n15 = Look At - Bone: dyn_R_hairFront_Lookat(ComponentPose=@n14, Alpha=1) {
    alphaBoolBlend {blendOption=Linear}
    boneToModify {boneName=dyn_R_hairFront_Lookat}
    lookAtLocation=<4.5448 3.7001 0>
    lookAtTarget {boneReference {boneName=dyn_Front_r_Hair_TGT}}
    lookAt_Axis {axis=<1 0>}
   }
   Save cached pose 'presim'(Pose=Component To Local(ComponentPose=@n15)) {cacheName=presim}
   n16 = RigidBody!(ComponentPose=Local To Component(LocalPose=Use cached pose 'presim'()), Alpha, OverrideWorldGravity:Vector="0,0,-980", ExternalForce) -> (Pose) {
    bClampLinearTranslationLimitToRefPose=true
    bOverrideWorldGravity=true
    overridePhysicsAsset=&/F_MED_Fuzzy_Bear_Summer_Hood_Up_Front_FaceAcc_Physics
    simulationSpace=BaseBoneSpace
    alphaBoolBlend {blendOption=Linear}
    baseBoneRef {boneName=spine_02}
    componentAppliedLinearAccClamp=<3500 3500 3500>
    componentLinearAccScale=<0.3 0.3 0.4>
    componentLinearVelScale=<0.3 0.3 0.4>
   }
   n17 = RigidBody(ComponentPose=Local To Component(LocalPose=Use cached pose 'presim'()), OverrideWorldGravity="0,0,-980") {
    bClampLinearTranslationLimitToRefPose=true
    bOverrideWorldGravity=true
    overridePhysicsAsset=&/F_MED_Fuzzy_Bear_Summer_Hood_Up_FrontEnd_FaceAcc_Physics
    simulationSpace=BaseBoneSpace
    alphaBoolBlend {blendOption=Linear}
    baseBoneRef {boneName=spine_02}
    componentAppliedLinearAccClamp=<3500 3500 3500>
    componentLinearAccScale=<0.3 0.3 0.4>
    componentLinearVelScale=<0.3 0.3 0.4>
   }
   Save cached pose 'post sim'(Pose=Blend Poses by bool(BlendPose_0=Component To Local(ComponentPose=@n17), BlendPose_1=Component To Local(ComponentPose=@n16), bActiveValue=Get bIsInFrontEnd()) {blendType=Linear}) {cacheName=post sim}
  ]}
 EventGraph {nodes [
   Event BlueprintUpdateAnimation()
   > n1 = Gravity Override State Function DEPRECATED(CharacterPartAnimInstance=Self-Reference(), GameTimeInSeconds=GetGameTimeInSeconds(), GravityOverrideParams=Get Gravity Override Params())
   > Set General Gravity Override(General Gravity Override:Vector=@n1) -> (Output_Get:Vector)
   n2 = AND Boolean(A=CheckForBackbling(), B=NOT Boolean(A=Get bIsInFrontEnd()))
   > Set backblingEquippedGravity(backblingEquippedGravity:Vector=SelectVector(A="0,0,-3000", B="0, 0,-1111", bPickA=@n2)) -> (Output_Get:Vector)
   > Set gravityFinal(gravityFinal:Vector=SelectVector(
     A=Get backblingEquippedGravity() -> (backblingEquippedGravity:Vector)
     B=Get General Gravity Override() -> (General Gravity Override:Vector)
     bPickA=@n2
    )) -> (Output_Get:Vector)
   > Set head rot(head rot:double=GetAlphaFromReferenceTransformDelta(SkeletalMeshComponent=Get Skel_Mesh(), BoneName=head, BaseName=spine_05, InputRange=(X=-20,Y=-50), TwistAxis="1,0,0")) -> (Output_Get:double)
   Event BlueprintInitializeAnimation()
   n3 = GetOwnerSkeletalMeshComponent()
   > Set Body_Mesh(Body_Mesh=@n3)
   > Set Skel_Mesh(Skel_Mesh=@n3)
  ]}
]
variables [
 backblingEquippedGravity {type=Vector}
 Body_Mesh {type=SkeletalMeshComponent}
 General Gravity Override {type=Vector}
 Gravity Override Params {type=GravityOverrideParamsStruct}
 gravityFinal {type=Vector}
 head rot {type=number}
 Skel_Mesh {type=SkeletalMeshComponent}
]
